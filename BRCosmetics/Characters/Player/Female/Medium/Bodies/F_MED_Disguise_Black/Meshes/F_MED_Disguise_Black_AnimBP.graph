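# Post-copy AnimBlueprint for the female medium Disguise Black body cosmetic that adds secondary motion to its left and right lace bones.
# It copies the pose from the attached parent mesh, then chains one spring-driven, angle-limited AnimDynamics node per lace bone, simulated in actor space.
# Additive bone-space Transform Bone nodes then apply a fixed pitch offset to each lace after simulation.
# The EventGraph is empty and no variables are defined.
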AnimBlueprint /BRCosmetics/Characters/Player/Female/Medium/Bodies/F_MED_Disguise_Black/Meshes/F_MED_Disguise_Black_AnimBP
target_skeleton=&/F_MED_Disguise_Black_Skeleton
graphs [
 AnimGraph {nodes [
   n1 = Anim Dynamics - Bone: dyn_L_lace_A1(ComponentPose=Local To Component(LocalPose=Copy Pose From Mesh() {bUseAttachedParent=true}), Alpha=1, ExternalForce="0,0,0") {
    angularDampingOverride=0.9999
    angularSpringConstant=40
    bAngularSpring=true
    bGravityOverrideInSimSpace=true
    bOverrideAngularDamping=true
    simulationSpace=Actor
    alphaBoolBlend {blendOption=Linear}
    boundBone {boneName=dyn_L_lace_A1}
    componentLinearAccScale=<1.5 1.5 1.5>
    physicsBodyDefinitions [{
      sphereCollisionRadius=0
      boundBone {boneName=dyn_L_lace_A1}
      boxExtents=<1 1 1>
      constraintSetup {coneAngle=20, angularLimitsMax=<0 30 20>, angularLimitsMin=<0 -10 -20>, angularTarget=<1 0 0>}
      localJointOffset=<0 0 0>
     }]
    relativeSpaceBone {boneName=calf_l}
   }
   n2 = Anim Dynamics - Bone: dyn_L_lace_A2(ComponentPose=@n1, Alpha=1, ExternalForce="0,0,0") {
    angularDampingOverride=0.9999
    angularSpringConstant=40
    bAngularSpring=true
    bGravityOverrideInSimSpace=true
    bOverrideAngularDamping=true
    simulationSpace=Actor
    alphaBoolBlend {blendOption=Linear}
    boundBone {boneName=dyn_L_lace_A2}
    componentLinearAccScale=<1.5 1.5 1.5>
    physicsBodyDefinitions [{
      sphereCollisionRadius=0
      boundBone {boneName=dyn_L_lace_A2}
      boxExtents=<1 1 1>
      constraintSetup {coneAngle=20, angularLimitsMax=<0 30 20>, angularLimitsMin=<0 -10 -20>, angularTarget=<1 0 0>}
      localJointOffset=<0 0 0>
     }]
    relativeSpaceBone {boneName=calf_l}
   }
   n3 = Anim Dynamics - Bone: dyn_L_lace_B1(ComponentPose=@n2, Alpha=1, ExternalForce="0,0,0") {
    angularDampingOverride=0.9999
    angularSpringConstant=40
    bAngularSpring=true
    bGravityOverrideInSimSpace=true
    bOverrideAngularDamping=true
    simulationSpace=Actor
    alphaBoolBlend {blendOption=Linear}
    boundBone {boneName=dyn_L_lace_B1}
    componentLinearAccScale=<1.5 1.5 1.5>
    physicsBodyDefinitions [{
      sphereCollisionRadius=0
      boundBone {boneName=dyn_L_lace_B1}
      boxExtents=<1 1 1>
      constraintSetup {coneAngle=20, angularLimitsMax=<0 30 20>, angularLimitsMin=<0 -10 -20>, angularTarget=<1 0 0>}
      localJointOffset=<0 0 0>
     }]
    relativeSpaceBone {boneName=calf_l}
   }
   n4 = Anim Dynamics - Bone: dyn_L_lace_B2(ComponentPose=@n3, Alpha=1, ExternalForce="0,0,0") {
    angularDampingOverride=0.9999
    angularSpringConstant=40
    bAngularSpring=true
    bGravityOverrideInSimSpace=true
    bOverrideAngularDamping=true
    simulationSpace=Actor
    alphaBoolBlend {blendOption=Linear}
    boundBone {boneName=dyn_L_lace_B2}
    componentLinearAccScale=<1.5 1.5 1.5>
    physicsBodyDefinitions [{
      sphereCollisionRadius=0
      boundBone {boneName=dyn_L_lace_B2}
      boxExtents=<1 1 1>
      constraintSetup {coneAngle=20, angularLimitsMax=<0 30 20>, angularLimitsMin=<0 -10 -20>, angularTarget=<1 0 0>}
      localJointOffset=<0 0 0>
     }]
    relativeSpaceBone {boneName=calf_l}
   }
   n5 = Anim Dynamics - Bone: dyn_R_lace_A1(ComponentPose=@n4, Alpha=1, ExternalForce="0,0,0") {
    angularDampingOverride=0.9999
    angularSpringConstant=40
    bAngularSpring=true
    bGravityOverrideInSimSpace=true
    bOverrideAngularDamping=true
    simulationSpace=Actor
    alphaBoolBlend {blendOption=Linear}
    boundBone {boneName=dyn_R_lace_A1}
    componentLinearAccScale=<1.5 1.5 1.5>
    physicsBodyDefinitions [{
      sphereCollisionRadius=0
      boundBone {boneName=dyn_R_lace_A1}
      boxExtents=<1 1 1>
      constraintSetup {coneAngle=20, angularLimitsMax=<0 30 20>, angularLimitsMin=<0 -10 -20>, angularTarget=<1 0 0>}
      localJointOffset=<0 0 0>
     }]
    relativeSpaceBone {boneName=calf_r}
   }
   n6 = Anim Dynamics - Bone: dyn_R_lace_A2(ComponentPose=@n5, Alpha=1, ExternalForce="0,0,0") {
    angularDampingOverride=0.9999
    angularSpringConstant=40
    bAngularSpring=true
    bGravityOverrideInSimSpace=true
    bOverrideAngularDamping=true
    simulationSpace=Actor
    alphaBoolBlend {blendOption=Linear}
    boundBone {boneName=dyn_R_lace_A2}
    componentLinearAccScale=<1.5 1.5 1.5>
    physicsBodyDefinitions [{
      sphereCollisionRadius=0
      boundBone {boneName=dyn_R_lace_A2}
      boxExtents=<1 1 1>
      constraintSetup {coneAngle=20, angularLimitsMax=<0 30 20>, angularLimitsMin=<0 -10 -20>, angularTarget=<1 0 0>}
      localJointOffset=<0 0 0>
     }]
    relativeSpaceBone {boneName=calf_r}
   }
   n7 = Anim Dynamics - Bone: dyn_R_lace_B1(ComponentPose=@n6, Alpha=1, ExternalForce="0,0,0") {
    angularDampingOverride=0.9999
    angularSpringConstant=40
    bAngularSpring=true
    bGravityOverrideInSimSpace=true
    bOverrideAngularDamping=true
    simulationSpace=Actor
    alphaBoolBlend {blendOption=Linear}
    boundBone {boneName=dyn_R_lace_B1}
    componentLinearAccScale=<1.5 1.5 1.5>
    physicsBodyDefinitions [{
      sphereCollisionRadius=0
      boundBone {boneName=dyn_R_lace_B1}
      boxExtents=<1 1 1>
      constraintSetup {coneAngle=20, angularLimitsMax=<0 30 20>, angularLimitsMin=<0 -10 -20>, angularTarget=<1 0 0>}
      localJointOffset=<0 0 0>
     }]
    relativeSpaceBone {boneName=calf_r}
   }
   n8 = Anim Dynamics - Bone: dyn_R_lace_B2(ComponentPose=@n7, Alpha=1, ExternalForce="0,0,0") {
    angularDampingOverride=0.9999
    angularSpringConstant=40
    bAngularSpring=true
    bGravityOverrideInSimSpace=true
    bOverrideAngularDamping=true
    simulationSpace=Actor
    alphaBoolBlend {blendOption=Linear}
    boundBone {boneName=dyn_R_lace_B2}
    componentLinearAccScale=<1.5 1.5 1.5>
    physicsBodyDefinitions [{
      sphereCollisionRadius=0
      boundBone {boneName=dyn_R_lace_B2}
      boxExtents=<1 1 1>
      constraintSetup {coneAngle=20, angularLimitsMax=<0 30 20>, angularLimitsMin=<0 -10 -20>, angularTarget=<1 0 0>}
      localJointOffset=<0 0 0>
     }]
    relativeSpaceBone {boneName=calf_r}
   }
   n9 = "Transform (Modify) Bone - Bone: dyn_L_lace_A1"!(ComponentPose=@n8, Alpha=1, Translation="0,0,0", Scale="1,1,1") -> (Pose) {
    rotationMode=BMM_Additive
    rotationSpace=BCS_BoneSpace
    alphaBoolBlend {blendOption=Linear}
    boneToModify {boneName=dyn_L_lace_A1}
    rotation {pitch=-20}
   }
   n10 = "Transform (Modify) Bone - Bone: dyn_L_lace_A2"!(ComponentPose=@n9, Alpha=1, Translation="0,0,0", Scale="1,1,1") -> (Pose) {
    rotationMode=BMM_Additive
    rotationSpace=BCS_BoneSpace
    alphaBoolBlend {blendOption=Linear}
    boneToModify {boneName=dyn_L_lace_A2}
    rotation {pitch=-20}
   }
   n11 = "Transform (Modify) Bone - Bone: dyn_L_lace_B1"!(ComponentPose=@n10, Alpha=1, Translation="0,0,0", Scale="1,1,1") -> (Pose) {
    rotationMode=BMM_Additive
    rotationSpace=BCS_BoneSpace
    alphaBoolBlend {blendOption=Linear}
    boneToModify {boneName=dyn_L_lace_B1}
    rotation {pitch=-20}
   }
   n12 = "Transform (Modify) Bone - Bone: dyn_L_lace_B2"!(ComponentPose=@n11, Alpha=1, Translation="0,0,0", Scale="1,1,1") -> (Pose) {
    rotationMode=BMM_Additive
    rotationSpace=BCS_BoneSpace
    alphaBoolBlend {blendOption=Linear}
    boneToModify {boneName=dyn_L_lace_B2}
    rotation {pitch=-20}
   }
   n13 = "Transform (Modify) Bone - Bone: dyn_R_lace_A1"!(ComponentPose=@n12, Alpha=1, Translation="0,0,0", Scale="1,1,1") -> (Pose) {
    rotationMode=BMM_Additive
    rotationSpace=BCS_BoneSpace
    alphaBoolBlend {blendOption=Linear}
    boneToModify {boneName=dyn_R_lace_A1}
    rotation {pitch=-20}
   }
   n14 = "Transform (Modify) Bone - Bone: dyn_R_lace_A2"!(ComponentPose=@n13, Alpha=1, Translation="0,0,0", Scale="1,1,1") -> (Pose) {
    rotationMode=BMM_Additive
    rotationSpace=BCS_BoneSpace
    alphaBoolBlend {blendOption=Linear}
    boneToModify {boneName=dyn_R_lace_A2}
    rotation {pitch=-20}
   }
   n15 = "Transform (Modify) Bone - Bone: dyn_R_lace_B1"!(ComponentPose=@n14, Alpha=1, Translation="0,0,0", Scale="1,1,1") -> (Pose) {
    rotationMode=BMM_Additive
    rotationSpace=BCS_BoneSpace
    alphaBoolBlend {blendOption=Linear}
    boneToModify {boneName=dyn_R_lace_B1}
    rotation {pitch=-20}
   }
   n16 = "Transform (Modify) Bone - Bone: dyn_R_lace_B2"!(ComponentPose=@n15, Alpha=1, Translation="0,0,0", Scale="1,1,1") -> (Pose) {
    rotationMode=BMM_Additive
    rotationSpace=BCS_BoneSpace
    alphaBoolBlend {blendOption=Linear}
    boneToModify {boneName=dyn_R_lace_B2}
    rotation {pitch=-20}
   }
   Output Pose(Result=Component To Local(ComponentPose=@n16)) {layerGroup=None}
  ]}
 EventGraph {links [], nodes []}
]
variables []
