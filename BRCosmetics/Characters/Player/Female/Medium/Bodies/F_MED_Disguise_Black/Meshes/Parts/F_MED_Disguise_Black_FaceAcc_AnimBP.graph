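# Post-process-style secondary motion AnimBlueprint for the hood bones of the F_MED_Disguise_Black FaceAcc part.
# It copies the attached parent's pose, then chains spring-based Anim Dynamics on six top, back, left and right hood bones.
# Right upper-arm yaw, measured against the clavicle, fades out dyn_R_hood_2 dynamics and blends in a replacement translation.
# Above LOD 0, additive translations nudge the top and first side hood joints.
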AnimBlueprint /BRCosmetics/Characters/Player/Female/Medium/Bodies/F_MED_Disguise_Black/Meshes/Parts/F_MED_Disguise_Black_FaceAcc_AnimBP
target_skeleton=&/F_MED_Disguise_Black_FaceAcc_Skeleton
graphs [
 AnimGraph {nodes [
   n1 = Anim Dynamics - Bone: dyn_top_hood(ComponentPose=Local To Component(LocalPose=Copy Pose From Mesh() {bUseAttachedParent=true}), Alpha=1, ExternalForce="0,0,0") {
    angularDampingOverride=0.99
    angularSpringConstant=200
    bAngularSpring=true
    bGravityOverrideInSimSpace=true
    bLinearSpring=true
    bOverrideLinearDamping=true
    gravityScale=0.7
    linearDampingOverride=0.9
    linearSpringConstant=50
    simulationSpace=RootRelative
    alphaBoolBlend {blendOption=Linear}
    boundBone {boneName=dyn_top_hood}
    componentLinearAccScale=<2 2 1>
    physicsBodyDefinitions [{
      sphereCollisionRadius=0
      boundBone {boneName=dyn_top_hood}
      boxExtents=<5 5 1>
      constraintSetup {angularLimitsMax=<0 0 0>, angularLimitsMin=<-50 0 0>, angularTarget=<0 0 0>, linearAxesMax=<0 0 3>}
      localJointOffset=<0 0 0>
     }]
   }
   n2 = Anim Dynamics - Bone: dyn_L_hood_1(ComponentPose=@n1, Alpha=1, ExternalForce="0,0,0") {
    bGravityOverrideInSimSpace=true
    bLinearSpring=true
    bOverrideLinearDamping=true
    linearDampingOverride=0.999
    linearSpringConstant=50
    alphaBoolBlend {blendOption=Linear}
    boundBone {boneName=dyn_L_hood_1}
    componentLinearAccScale=<1 1 1>
    physicsBodyDefinitions [{
      sphereCollisionRadius=0
      boundBone {boneName=dyn_L_hood_1}
      boxExtents=<6 5 3>
      constraintSetup {angularLimitsMax=<0 0 0>, angularLimitsMin=<0 0 0>, angularTarget=<1 0 0>, linearAxesMax=<1.5 0 0>}
      localJointOffset=<0 0 0>
     }]
   }
   n3 = Anim Dynamics - Bone: dyn_R_hood_1(ComponentPose=@n2, Alpha=1, ExternalForce="0,0,0") {
    bGravityOverrideInSimSpace=true
    bLinearSpring=true
    bOverrideLinearDamping=true
    linearDampingOverride=0.999
    linearSpringConstant=50
    alphaBoolBlend {blendOption=Linear}
    boundBone {boneName=dyn_R_hood_1}
    componentLinearAccScale=<1 1 1>
    physicsBodyDefinitions [{
      sphereCollisionRadius=0
      boundBone {boneName=dyn_R_hood_1}
      boxExtents=<6 5 3>
      constraintSetup {angularLimitsMax=<0 0 0>, angularLimitsMin=<0 0 0>, angularTarget=<-1 0 0>, linearAxesMin=<-1.5 0 0>}
      localJointOffset=<0 0 0>
     }]
   }
   n4 = Anim Dynamics - Bone: dyn_back_hood(ComponentPose=@n3, Alpha=1, ExternalForce="0,0,0") {
    angularSpringConstant=90
    bAngularSpring=true
    bGravityOverrideInSimSpace=true
    bLinearSpring=true
    bOverrideLinearDamping=true
    gravityScale=0.8
    linearDampingOverride=0.9
    linearSpringConstant=70
    simulationSpace=World
    alphaBoolBlend {blendOption=Linear}
    boundBone {boneName=dyn_back_hood}
    componentLinearAccScale=<1 1 0.3>
    physicsBodyDefinitions [{
      sphereCollisionRadius=0
      boundBone {boneName=dyn_back_hood}
      boxExtents=<6 5 3>
      constraintSetup {angularLimitsMax=<10 10 10>, angularLimitsMin=<-10 -10 -10>, angularTarget=<0 0 0>, linearAxesMax=<3 1 3>, linearAxesMin=<-3 -4 0>}
      localJointOffset=<0 0 0>
     }]
   }
   n5 = Anim Dynamics - Bone: dyn_L_hood_2(ComponentPose=@n4, Alpha=1, ExternalForce="0,0,0") {
    bGravityOverrideInSimSpace=true
    bLinearSpring=true
    bOverrideLinearDamping=true
    linearDampingOverride=0.999
    linearSpringConstant=50
    alphaBoolBlend {blendOption=Linear}
    boundBone {boneName=dyn_L_hood_2}
    componentLinearAccScale=<1 1 1>
    physicsBodyDefinitions [{
      sphereCollisionRadius=0
      boundBone {boneName=dyn_L_hood_2}
      boxExtents=<6 5 3>
      constraintSetup {angularLimitsMax=<0 0 0>, angularLimitsMin=<0 0 0>, angularTarget=<1 0 0>, linearAxesMax=<3 0 0>, linearAxesMin=<-3 0 0>}
      localJointOffset=<0 0 0>
     }]
   }
   n6 = Anim Dynamics - Bone: dyn_R_hood_2(ComponentPose=@n5, Alpha=Get turnDynamicsOff() -> (turnDynamicsOff:double), ExternalForce="0,0,0") {
    bGravityOverrideInSimSpace=true
    bLinearSpring=true
    bOverrideLinearDamping=true
    linearDampingOverride=0.999
    linearSpringConstant=50
    alphaBoolBlend {blendOption=Linear}
    alphaScaleBiasClamp {bInterpResult=true, interpSpeedDecreasing=5, interpSpeedIncreasing=5}
    boundBone {boneName=dyn_R_hood_2}
    componentLinearAccScale=<1 1 1>
    physicsBodyDefinitions [{
      sphereCollisionRadius=0
      boundBone {boneName=dyn_R_hood_2}
      boxExtents=<6 5 3>
      constraintSetup {angularLimitsMax=<0 0 0>, angularLimitsMin=<0 0 0>, angularTarget=<-1 0 0>, linearAxesMax=<3 0 0>, linearAxesMin=<-3 0 0>}
      localJointOffset=<0 0 0>
     }]
   }
   n7 = "Transform (Modify) Bone - Bone: dyn_R_hood_2"(ComponentPose=@n6, Alpha=Get clavRotateAlpha() -> (clavRotateAlpha:double), Translation="9.5,0,0", Rotation="0,0,0", Scale="1,1,1") {
    translationMode=BMM_Replace
    translationSpace=BCS_BoneSpace
    alphaBoolBlend {blendOption=Linear}
    alphaScaleBiasClamp {bInterpResult=true, interpSpeedDecreasing=20, interpSpeedIncreasing=20}
    boneToModify {boneName=dyn_R_hood_2}
    translation {x=9.5}
   }
   n8 = Get scootJoint() -> (scootJoint:double)
   n9 = "Transform (Modify) Bone - Bone: dyn_top_hood"(ComponentPose=@n7, Alpha=@n8, Translation="0,0,2", Rotation="0,0,0", Scale="1,1,1") {translationMode=BMM_Additive, alphaBoolBlend {blendOption=Linear}, boneToModify {boneName=dyn_top_hood}, translation {z=2}}
   n10 = "Transform (Modify) Bone - Bone: dyn_L_hood_1"(ComponentPose=@n9, Alpha=@n8, Translation="1.5,0,0", Rotation="0,0,0", Scale="1,1,1") {translationMode=BMM_Additive, alphaBoolBlend {blendOption=Linear}, boneToModify {boneName=dyn_L_hood_1}, translation {x=1.5}}
   n11 = "Transform (Modify) Bone - Bone: dyn_R_hood_1"(ComponentPose=@n10, Alpha=@n8, Translation="-1.5,0,0", Rotation="0,0,0", Scale="1,1,1") {translationMode=BMM_Additive, alphaBoolBlend {blendOption=Linear}, boneToModify {boneName=dyn_R_hood_1}, translation {x=-1.5}}
   Output Pose(Result=Component To Local(ComponentPose=@n11)) {layerGroup=None}
  ]}
 EventGraph {nodes [
   Event BlueprintUpdateAnimation()
   > n1 = Sequence!(execute) -> (then_0, then_1, then_2:Exec)
   n2 = BreakTransform(InTransform=GetDeltaTransformFromRefPose(self=GetOwnerSkeletalMeshComponent(), BoneName=upperarm_r, BaseName=clavicle_r))
   Set clavRotateAlpha(execute=@n1.then_1, clavRotateAlpha:double=MapRangeClamped(Value=BreakRotator(InRot=@n2.Rotation).Yaw, InRangeA=70, InRangeB=100)) -> (Output_Get:double)
   n3 = Get CurrentLOD()
   Set scootJoint(execute=@n1.then_2, scootJoint:double="To Float (Boolean)"(InBool="integer > integer"(A=@n3, B=0) {is=title K2Node_CallFunction})) -> (Output_Get:double)
   n4 = Get clavRotateAlpha()
   Set turnDynamicsOff(execute=@n1.then_0, turnDynamicsOff:double=SelectFloat(A=0, B=1, bPickA="float >= float"(A=@n4, B=0.5) {is=title K2Node_CallFunction})) -> (Output_Get:double)
  ]}
]
variables [
 clavRotateAlpha {type=number}
 scootJoint {type=number}
 turnDynamicsOff {type=number}
]
